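# Additive Niagara sprite and mesh-particle material for the impact effect of the female Cactus Dancer pickaxe.
# Particle color multiplies the T_CactusDancer_Impact texture and passes through MF_FX_EmissiveAdjustments.
# Separate branches hold step thresholds, a radial-gradient curve-atlas lookup and a lerp driven by a Flipbook dynamic parameter.
# These branches do not feed the emissive function and appear unused.
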Material /BRCosmetics/Weapons/FORT_Melee/Pickaxe_Cactus_Dancer_Female/FX/M_CactusDancer_Impact
graph {
 properties {
  bUsedWithNiagaraMeshParticles=true
  bUsedWithNiagaraSprites=true
  blendMode=BLEND_Additive
  pixelDepthOffsetMode=PDOM_Legacy
  usedShadingModels=Default Lit
 }
 nodes [
  n1 = TextureSample() {texture=&/T_CactusDancer_Impact_02}
  Add(A=@n1.RGB, B=ScalarParameter())
  Constant()
  CurveAtlasRowParameter(CurveTime=MaterialFunctionCall(Density=ScalarParameter() {defaultValue=1, parameterName=Param_2}) {materialFunction=&RadialGradientExponential}.RadialGradientExponential) {
   atlas=&AthenaCurveAtlas01
   curve=/Game/Athena/LinearColorCurves/Curve_Skydive_Sony2
   defaultValue=90
   parameterName=Param_1
   sortPriority=0
  }
  MaterialFunctionCall!(ThresholdValue, ThresholdRange, Alpha, TwoSidedThreshold) {materialFunction=/Engine/Functions/Engine_MaterialFunctions02/Masking/ThresholdWithRange}
  n2 = TextureSample() {texture=&/T_CactusDancer_Impact}
  MaterialFunctionCall(Emissive=Multiply(A=ParticleColor().RGB, B=@n2.RGB)) {materialFunction=&MF_FX_EmissiveAdjustments}
  n3 = TextureSample() {texture=&/T_FX_CactusDancer}
  n4 = MaterialFunctionCall(
   1=@n3.R
   2=@n3.R
   3=@n3.R
   4=Constant()
   A=DynamicParameter() {defaultValue=<rgba 0 0 0 0>, paramNames [Flipbook,Param2,Param3,Param4]}.Flipbook
  ) {materialFunction=&Lerp_Multiple_Float3}
  Multiply(A=@n4.Lerp 4 Inputs, B=@n4.Lerp 4 Inputs)
  Multiply()
  OneMinus(None=Step(Y=@n2.R, X=ScalarParameter() {parameterName=Param_4}))
  Step(Y=@n1.R, X=ScalarParameter() {parameterName=Param_3})
 ]
}
