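# Additive Niagara sprite material for the impact effect of the female Cactus Dancer pickaxe.
# An Erosion dynamic parameter thresholds one texture's alpha into a hard-edged mask, which multiplies a second, saturation-adjusted texture.
# Particle color tints the result, which feeds an emissive adjustment function, while particle alpha is multiplied by the same mask.
# Several nodes, including a third texture and a subtract-based erosion, are left unconnected.
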Material /BRCosmetics/Weapons/FORT_Melee/Pickaxe_Cactus_Dancer_Female/FX/M_CactusDancer_Impactt
graph {
 properties {bUsedWithNiagaraSprites=true, blendMode=BLEND_Additive, pixelDepthOffsetMode=PDOM_Legacy, usedShadingModels=Default Lit}
 nodes [
  DepthFade(Opacity=Constant() {r=1}, FadeDistance=ScalarParameter() {defaultValue=40, parameterName=FadeDistance})
  n1 = ParticleColor()
  n2 = TextureSample() {samplerType=SAMPLERTYPE_LinearColor, texture=&/T_FX_CactusDancer_Impact_03}
  n3 = Desaturation(None=@n2.RGB, Fraction=ScalarParameter() {defaultValue=-0.2, parameterName=Param_2})
  n4 = TextureSample() {samplerType=SAMPLERTYPE_LinearColor, texture=&/T_FX_CactusDancer_Impact}
  n5 = DynamicParameter() {defaultValue=<rgba 0 1 1 1>, paramNames [Erosion,Param2,Param3,Param4]}
  n6 = OneMinus(None=Step(Y=@n4.A, X=@n5.Erosion))
  MaterialFunctionCall(Emissive=Multiply(A=@n1.RGB, B=Multiply(A=@n3, B=@n6))) {materialFunction=&MF_FX_EmissiveAdjustments}
  Multiply(A=@n1.A, B=@n6)
  ScalarParameter()
  ScalarParameter() {defaultValue=0.1, parameterName=Param_1}
  Subtract(A=@n4.A, B=@n5.Erosion)
  TextureSample() {samplerType=SAMPLERTYPE_LinearColor, texture=&/T_FX_CactusDancer_Impact_02}
 ]
}
